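# Niagara effect for the Constellation backpack cosmetic, a looping cyan-teal glow with pulsing meshes, flares and sparkles.
# Base and Base_SMALL emit rotated meshes that swell and fade, over BG sprites whose RGB is zeroed and Flare sprites that flash bright.
# Sparkles spawn on a small flipped hemisphere and drift outward, starting overbright.
# Glow and Light002 pair sprites with light renderers, Light002 cycling a persistent particle, while GlowStart is disabled.
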NiagaraSystem /BRCosmetics/Athena/Effects/Particles/BackPack/NS_Constellation_Backpack
dependencies {
 usedDataInterfaces [&NiagaraDataInterfaceCurve, &NiagaraDataInterfaceVectorCurve, &NiagaraDataInterfaceColorCurve]
 usedDynamicInputs [&Vector2DFromFloat]
 usedModules [&SystemState]
}
emitters {
 BG {
  emitterUpdateScript [
   &EmitterState(Life Cycle Mode=::NewEnumerator1, Loop Duration=1.5, UseLoopDelay=true)
   &SpawnBurst_Instantaneous(Use Spawn Probability=false, UseLoopCountLimit=false)
  ]
  particleSpawnScript [
   &InitializeParticle(
    Lifetime=3
    Position Offset=<0 0 -5>
    Sprite Rotation Mode=::NewEnumerator2
    Sprite Size Mode=::NewEnumerator3
    Uniform Sprite Size=75
    UsePositionOffset=true
    Write Material Random=true
   ) {module=InitializeParticle001}
   &Color(Color=&MakeLinearColorFromVectorAndFloat("Vector (RGB)"=<0 0.4 0.3>, "Float (Alpha)"=&ClampFloat(Float=1.5)))
  ]
  particleUpdateScript [
   &ParticleState()
   &CameraOffset(Camera Offset Amount=15)
   &ScaleColor(
    Scale Alpha=&FloatFromCurve(FloatCurve=DataInterface() {propertyValues {Curve {keys [
         <at 0.0011 0.0003 RCIM_Cubic>
         {arriveTangent=-0.0003, interpMode=RCIM_Cubic, leaveTangent=-0.0003, time=0.4951, value=0.7894}
         {interpMode=RCIM_Cubic, leaveTangent=-1.3318, time=1, value=0}
        ]}}})
    Scale RGB=<0 0 0>
    ScaleA=true
    ScaleRGB=true
   )
   &ScaleSpriteSize()
   &DynamicMaterialParameters(
    Index 0 Param 1=1.5
    Index 0 Param 3=1
    Index 0 Param 4=1
    Module.Param0WriteEnabled=true
    Module.Param1WriteEnabled=true
    Module.Param2WriteEnabled=true
    Module.Param3WriteEnabled=true
   )
  ]
  rendererClasses [&NiagaraSpriteRendererProperties]
  renderers [{rendererClass=&NiagaraSpriteRendererProperties}]
 }
 Base {
  emitterUpdateScript [
   &EmitterState(Inactive Response=::NewEnumerator1, Life Cycle Mode=::NewEnumerator1, Loop Duration=1, UseLoopDelay=true)
   &SpawnRate(SpawnRate=2, Use Spawn Probability=false) {module=SpawnRate001}
   &SpawnBurst_Instantaneous(Use Spawn Probability=false, UseLoopCountLimit=false)
  ]
  particleSpawnScript [
   /SetVariables(Particles.MeshOrientation=&MakeQuatFromAxisAngle(Axis=<0 0.85 0>, AngleInDegrees=90)) {isSetParameters=true}
   /Niagara/Modules/Spawn/Initialization/InitializeParticle(
    Mesh Scale=<0.45 0.45 0.45>
    Module.Write Color=true
    Module.Write Lifetime=true
    Module.Write Mass=true
    Module.Write Material Random=true
    Module.Write Position=true
    Module.Write Scale=true
    Module.Write SpriteRotation=false
    Module.Write SpriteSize=false
    Position=<0 0 -5>
   )
   &Color(Color=&MakeLinearColorFromVectorAndFloat("Vector (RGB)"=<0 1.25 1>, "Float (Alpha)"=&ClampFloat(Float=0.2)))
  ]
  particleUpdateScript [
   &ParticleState()
   &ScaleColor(
    Scale Alpha=&FloatFromCurve(FloatCurve=DataInterface() {propertyValues {Curve {keys [<at 0 0 RCIM_Cubic>,<at 0.3204 0.9875 RCIM_Cubic>,<at 1 0>]}}})
    ScaleA=true
    ScaleRGB=true
   )
   &ScaleMeshSize(Scale Factor=&VectorFromFloat(Value=&FloatFromCurve(FloatCurve=DataInterface() {propertyValues {Curve {keys [<at 0 0.8>,<at 0.5 1>,<at 1 0.8>]}}})))
  ]
  rendererClasses [&NiagaraMeshRendererProperties]
  renderers [{rendererClass=&NiagaraMeshRendererProperties, rendererIndex=0}]
 }
 Base_SMALL {
  emitterUpdateScript [
   &EmitterState(Inactive Response=::NewEnumerator1, Life Cycle Mode=::NewEnumerator1, Loop Duration=1, UseLoopDelay=true)
   &SpawnBurst_Instantaneous(Use Spawn Probability=false, UseLoopCountLimit=false)
   &SpawnBurst_Instantaneous(Spawn Time=0.5, Use Spawn Probability=false, UseLoopCountLimit=false) {module=SpawnBurst_Instantaneous001}
  ]
  particleSpawnScript [
   /SetVariables(Particles.MeshOrientation=&MakeQuatFromAxisAngle(Axis=<0 0.85 0>, AngleInDegrees=90)) {isSetParameters=true}
   &InitializeParticle(
    Lifetime=1.1
    Mesh Scale=/Niagara/DynamicInputs/CascadeConversion/SplitVector(
     Double Component Vector Channels:ECascadeNiagaraTwoVectorChannels=::NewEnumerator1
     Double Component:NiagaraFloat=0.32
     Single Component:NiagaraFloat=0.27
    )
    Module.Write Color=true
    Module.Write Lifetime=true
    Module.Write Mass=true
    Module.Write Material Random=true
    Module.Write Position=true
    Module.Write Scale=true
    Module.Write SpriteRotation=false
    Module.Write SpriteSize=false
    Position=<0 0 -5>
   )
   &Color(Color=&MakeLinearColorFromVectorAndFloat("Vector (RGB)"=<0.1 0.6 0.6>, "Float (Alpha)"=&ClampFloat(Float=1)))
  ]
  particleUpdateScript [
   &ParticleState()
   &ScaleColor(
    Scale Alpha=&FloatFromCurve(FloatCurve=DataInterface() {propertyValues {Curve {keys [
         <at -0.0014 0.0182 RCIM_Cubic>
         {arriveTangent=1.3859, interpMode=RCIM_Cubic, leaveTangent=1.3859, time=0.1499, value=0.9875}
         {arriveTangent=-1.1616, interpMode=RCIM_Cubic, leaveTangent=-1.1616, time=0.705, value=0.9971}
         <at 1 0>
        ]}}})
    ScaleA=true
    ScaleRGB=true
   )
   &ScaleMeshSize(Scale Factor=&VectorFromFloat(Value=&FloatFromCurve(FloatCurve=DataInterface() {propertyValues {Curve {keys [<at 0 0.8>,<at 0.5 1>,<at 1 0.8>]}}})))
  ]
  rendererClasses [&NiagaraMeshRendererProperties]
  renderers [{rendererClass=&NiagaraMeshRendererProperties, rendererIndex=0}]
 }
 Flare {
  emitterUpdateScript [
   &EmitterState(Life Cycle Mode=::NewEnumerator1, Loop Duration=0.5, UseLoopDelay=true)
   &SpawnBurst_Instantaneous(Use Spawn Probability=false, UseLoopCountLimit=false)
  ]
  particleSpawnScript [
   /Niagara/Modules/Spawn/Initialization/V2/InitializeParticle(
    Position Offset=<0 0 -5>
    Sprite Rotation Mode=::NewEnumerator2
    Sprite Size Mode=::NewEnumerator1
    Uniform Sprite Size Max=90
    Uniform Sprite Size Min=70
    UsePositionOffset=true
    Write Material Random=true
   ) {module=InitializeParticle001}
   &Color(Color=&MakeLinearColorFromVectorAndFloat("Vector (RGB)"=<0 1 0.9929>, "Float (Alpha)"=&ClampFloat(Float=0.5)))
  ]
  particleUpdateScript [
   &ParticleState()
   &CameraOffset(Camera Offset Amount=15)
   &ScaleColor(
    Scale Alpha=&FloatFromCurve(FloatCurve=DataInterface() {propertyValues {Curve {keys [
         <at 0.0011 0.0003 RCIM_Cubic>
         {arriveTangent=-0.0003, interpMode=RCIM_Cubic, leaveTangent=-0.0003, time=0.4951, value=0.7894}
         {interpMode=RCIM_Cubic, leaveTangent=-1.3318, time=1, value=0}
        ]}}})
    Scale RGB=&VectorFromFloat(Value=&FloatFromCurve(FloatCurve=DataInterface() {propertyValues {Curve {keys [<at 0 10 RCIM_Cubic>,<at 0.05 1>]}}}))
    ScaleA=true
    ScaleRGB=true
   )
   &ScaleSpriteSize(Uniform Curve Scale=1.2)
   &DynamicMaterialParameters(
    Index 0 Param 1=1
    Index 0 Param 3=1
    Index 0 Param 4=1
    Module.Param0WriteEnabled=true
    Module.Param1WriteEnabled=true
    Module.Param2WriteEnabled=true
    Module.Param3WriteEnabled=true
   )
  ]
  rendererClasses [&NiagaraSpriteRendererProperties]
  renderers [{rendererClass=&NiagaraSpriteRendererProperties}]
 }
 Glow {
  emitterUpdateScript [
   &EmitterState(Life Cycle Mode=::NewEnumerator1, Loop Duration=1, UseLoopDelay=true)
   &SpawnBurst_Instantaneous(Use Spawn Probability=false, UseLoopCountLimit=false)
  ]
  particleSpawnScript [
   /Niagara/Modules/Spawn/Initialization/InitializeParticle(
    Module.Write Color=true
    Module.Write Lifetime=true
    Module.Write Mass=true
    Module.Write Material Random=true
    Module.Write Position=true
    Module.Write Scale=false
    Module.Write SpriteRotation=false
    Module.Write SpriteSize=true
    Position=<0 0 -5>
    Sprite Size=<90 90>
   )
   &Color(Color=&MakeLinearColorFromVectorAndFloat("Vector (RGB)"=<0 1 0.9929>, "Float (Alpha)"=&ClampFloat(Float=0.5)))
  ]
  particleUpdateScript [
   /SetVariables(Particles.LightExponent=8) {isSetParameters=true}
   /SetVariables(Particles.LightColor:Vector3f=<1 1 1>) {isSetParameters=true}
   /SetVariables(Particles.LightRadius=32) {isSetParameters=true}
   &ParticleState()
  ]
  rendererClasses [&NiagaraSpriteRendererProperties, &NiagaraLightRendererProperties]
  renderers [
   {rendererClass=&NiagaraSpriteRendererProperties}
   {rendererClass=&NiagaraLightRendererProperties, rendererIndex=1}
  ]
 }
 GlowStart {
  bEnabled=false
  emitterUpdateScript [
   &EmitterState(Life Cycle Mode=::NewEnumerator1, Loop Duration=0.5, UseLoopDelay=true)
   &SpawnBurst_Instantaneous(Use Spawn Probability=false, UseLoopCountLimit=false)
  ]
  particleSpawnScript [
   &InitializeParticle(
    Mesh Scale=<0.5 0.5 0.5>
    Module.Write Color=true
    Module.Write Lifetime=true
    Module.Write Mass=true
    Module.Write Material Random=true
    Module.Write Position=true
    Module.Write Scale=true
    Module.Write SpriteRotation=false
    Module.Write SpriteSize=false
    Position=<0 0 -5>
   )
   &Color(Color=&MakeLinearColorFromVectorAndFloat("Vector (RGB)"=<0 1 0.9801>, "Float (Alpha)"=&ClampFloat(Float=0.005)))
  ]
  particleUpdateScript [
   &ParticleState()
   &CameraOffset(Camera Offset Amount=25)
   &DynamicMaterialParameters(
    Index 0 Param 1=&FloatFromCurve(FloatCurve=DataInterface() {propertyValues {Curve {keys [<at 0 0.5>,<at 1 0.3>]}}})
    Index 0 Param 2=2
    Index 0 Param 3=15
    Index 0 Param 4=1
    Module.Param0WriteEnabled=true
    Module.Param1WriteEnabled=true
    Module.Param2WriteEnabled=true
    Module.Param3WriteEnabled=true
   )
   &ScaleColor(
    Scale RGB=&VectorFromCurve(VectorCurve=DataInterface() {propertyValues {
       XCurve {keys [<at 0 0 RCIM_Cubic>,<at 0.4995 0.0001 RCIM_Cubic>,<at 1 0 RCIM_Cubic>]}
       YCurve {keys [<at 0 0 RCIM_Cubic>,<at 0.4995 0.5562 RCIM_Cubic>,<at 1 0 RCIM_Cubic>]}
       ZCurve {keys [<at 0 0 RCIM_Cubic>,<at 0.4995 1 RCIM_Cubic>,<at 1 0 RCIM_Cubic>]}
      }})
    ScaleA=true
    ScaleRGB=true
   )
  ]
  rendererClasses [&NiagaraMeshRendererProperties]
  renderers [{rendererClass=&NiagaraMeshRendererProperties, rendererIndex=0}]
 }
 Light002 {
  emitterUpdateScript [
   &EmitterState(Life Cycle Mode=::NewEnumerator1, Loop Behavior=::NewEnumerator1, Loop Duration Mode=::NewEnumerator1, UseLoopDelay=false)
   &SpawnBurst_Instantaneous(Spawn Time=0.001, Use Spawn Probability=false, UseLoopCountLimit=false)
   &SpawnRate(SpawnRate=8, Use Spawn Probability=false) {enabled=false}
  ]
  particleSpawnScript [
   /Niagara/Modules/Spawn/Initialization/V2/InitializeParticle(
    Lifetime=3
    Position Offset=<-8 0 -5>
    Sprite Size Mode=::NewEnumerator3
    Uniform Sprite Size=60
    UsePositionOffset=true
    Write Material Random=true
   ) {module=InitializeParticle001}
   &Color(Color=&MakeLinearColorFromVectorAndFloat("Vector (RGB)"=<1 1 1>, "Float (Alpha)"=&ClampFloat(Float=1)))
   &BoxLocation(Box Offset=<10 0 -10>, Box Size=<0 0 0>, MaskSpawnGroup=false, Module.Override Seed=false, Rotate Box=false, Surface Only=false) {module=BoxLocation001}
  ]
  particleUpdateScript [
   &ParticleState(
    Kill Particles When Lifetime Has Elapsed=false
    Let Infinitely Lived Particles Die When Emitter Deactivates=true
    Loop Particles Lifetime=true
   )
   &ScaleColor(
    Scale Alpha=&FloatFromCurve(FloatCurve=DataInterface() {propertyValues {Curve {keys [<at 0 1>,<at 1 0.15>]}}})
    ScaleA=true
    ScaleRGB=true
   )
   /SetVariables(Particles.Gen5LightColor:LinearColor=&ScaleLinearColorByCurve(
     InitialValue=<rgba 0.0611 0.9502 1 1>
     ScaleCurve=DataInterface() {propertyValues {
       AlphaCurve {keys [
         {interpMode=RCIM_Cubic, tangentMode=RCTM_User, time=0, value=5}
         {interpMode=RCIM_Cubic, tangentMode=RCTM_User, time=0.2243, value=4}
         <at 1 5>
        ]}
       BlueCurve {keys [<at 0 1>,<at 1 1>]}
       GreenCurve {keys [<at 0 1>,<at 1 1>]}
       RedCurve {keys [<at 0 1>,<at 1 1>]}
      }}
     CurveIndex=$Particles.NormalizedAge
    )) {isSetParameters=true}
   /SetVariables(Particles.Gen5LightIntensity:NiagaraFloat=&FloatFromCurve(FloatCurve=DataInterface() {propertyValues {Curve {keys [<at 0 150 RCIM_Cubic>]}}})) {enabled=false, isSetParameters=true}
   &CameraOffset(Camera Offset Amount=30) {enabled=false}
   /SetVariables(Particles.LightColor=&ScaleLinearColorByCurve(
     InitialValue=<rgba 1 0.2753 0.0949 1>
     ScaleCurve=DataInterface() {propertyValues {
       AlphaCurve {keys [{interpMode=RCIM_Cubic, tangentMode=RCTM_User, time=0, value=1}, {interpMode=RCIM_Cubic, tangentMode=RCTM_User, time=0.5, value=5}, <at 1 1>]}
       BlueCurve {keys [<at 0 1>]}
       GreenCurve {keys [<at 0 1>]}
       RedCurve {keys [<at 0 1>]}
      }}
     CurveIndex=$Particles.NormalizedAge
    )) {enabled=false, isSetParameters=true}
  ]
  rendererClasses [&NiagaraSpriteRendererProperties, &NiagaraLightRendererProperties]
  renderers [
   {rendererClass=&NiagaraSpriteRendererProperties}
   {rendererClass=&NiagaraLightRendererProperties, rendererIndex=1}
  ]
 }
 sparkles {
  emitterUpdateScript [
   &EmitterState(Life Cycle Mode=::NewEnumerator1, Loop Duration=1, UseLoopDelay=true)
   &SpawnRate(SpawnRate=20, Use Spawn Probability=false) {module=SpawnRate001}
  ]
  particleSpawnScript [
   &InitializeParticle(
    Lifetime Max=1
    Lifetime Min=0.5
    Lifetime Mode=::NewEnumerator1
    Position Offset=<0 0 -5>
    Sprite Size Mode=::NewEnumerator1
    Uniform Sprite Size Max=14
    Uniform Sprite Size Min=4
    UsePositionOffset=true
    Write Material Random=true
   ) {module=InitializeParticle001}
   &Color(Color=&MakeLinearColorFromVectorAndFloat("Vector (RGB)"=<0 20 16>, "Float (Alpha)"=&ClampFloat(Float=1)))
   /Niagara/Modules/Spawn/Location/SphereLocation(
    Hemisphere Z=true
    MaskSpawnGroup=false
    Module.Override Randomness=false
    Non Uniform Scale=<1 1 -1>
    Offset=<0 0 3>
    Sphere Radius=4
    Surface Only=true
   )
   &AddVelocity(Velocity=&Multiply_VectorByFloat(Vector=$Output.SphereLocation.SphereVector, Float=5))
  ]
  particleUpdateScript [
   &ParticleState()
   &ScaleColor(
    Scale Alpha=&FloatFromCurve(FloatCurve=DataInterface() {propertyValues {Curve {keys [
         <at 0 0.0994 RCIM_Cubic>
         {arriveTangent=-0.0994, interpMode=RCIM_Cubic, leaveTangent=-0.0994, time=0.128, value=0.0373}
         <at 1 0 RCIM_Cubic>
        ]}}})
    Scale RGB=&VectorFromCurve(VectorCurve=DataInterface() {propertyValues {XCurve {keys [<at 0 5>,<at 0.4 1>]}, YCurve {keys [<at 0 150>,<at 0.4 1>]}, ZCurve {keys [<at 0 100>,<at 0.4 1>]}}})
    ScaleA=true
    ScaleRGB=true
   )
   &ScaleSpriteSize()
   &SolveForcesAndVelocity(Clamp Velocity=false, Limit Acceleration=false)
  ]
  rendererClasses [&NiagaraSpriteRendererProperties]
  renderers [{rendererClass=&NiagaraSpriteRendererProperties}]
 }
}
summary {
 systemName=NS_Constellation_Backpack
 emitters [
  {emitterName=Glow, rendererClasses [&NiagaraSpriteRendererProperties, &NiagaraLightRendererProperties]}
  {emitterName=BG, rendererClasses [&NiagaraSpriteRendererProperties]}
  {emitterName=Flare, rendererClasses [&NiagaraSpriteRendererProperties]}
  {bEnabled=false, emitterName=GlowStart, rendererClasses [&NiagaraMeshRendererProperties]}
  {emitterName=Base, rendererClasses [&NiagaraMeshRendererProperties]}
  {emitterName=Base_SMALL, rendererClasses [&NiagaraMeshRendererProperties]}
  {emitterName=sparkles, rendererClasses [&NiagaraSpriteRendererProperties]}
  {emitterName=Light002, rendererClasses [&NiagaraSpriteRendererProperties, &NiagaraLightRendererProperties]}
 ]
 userVariables []
}
